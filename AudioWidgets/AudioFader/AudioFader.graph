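# AudioFader is a UMG widget: a vertical fader with a dB text box above it for setting an audio level.
# Slider moves convert linear to dB through a float curve, update the text and broadcast OnValueChanged in dB.
# Committed text, bound on Construct, maps through the inverse curve to set the slider, then broadcasts the dB recomputed from the slider, while the text shows the typed number.
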WidgetBlueprint /AudioWidgets/AudioFader/AudioFader
graphs [
 EventGraph {nodes [
   n1 = "On Value Changed (Slider)"!() -> (OutputDelegate, then, Value:float) {is=title K2Node_ComponentBoundEvent}
   > n2 = GetDbValue!(execute, self, LinValue:double=@n1.Value) -> (then, DbValue:double)
   > SetTextValue!(execute, self, NewValue:double=@n2) -> (then)
   > Call On Value Changed!(execute:Exec, self:Audio Fader Object Reference, NewValue:double=@n2) -> (then:Exec) {is=title K2Node_CallDelegate}
   n3 = OnTextCommitted!() -> (OutputDelegate, then, Text:"Text (by ref)", CommitMethod:ETextCommit Enum=Default) {is=title K2Node_CustomEvent}
   n4 = String To Float(InString="To String (Text)"(InText=@n3.Text))
   > n5 = GetLinValue!(execute, self, DbValue:double=@n4) -> (then, LinValue:double)
   > SetTextValue(NewValue=@n4)
   > SetValue!(execute, self:Slider Object Reference=Get Slider() -> (Slider:Slider Object Reference), InValue:float=@n5) -> (then)
   > n6 = GetDbValue(LinValue=GetValue!(self:Slider Object Reference=Get Slider()) -> (ReturnValue:float))
   > Call On Value Changed(NewValue=@n6)
   Event Construct()
   n7 = Get AudioTextBox() -> (AudioTextBox:Audio Text Box Object Reference)
   > Set UnitText!(execute, UnitText:Text="LOCTEXT(\"dB\")", self:Audio Text Box Object Reference=@n7) -> (then, Output_Get:Text)
   > "Assign OnTextCommitted (Editable Text)"!(
    execute:Exec
    self:Editable Text Object Reference=Get ValueTextBox!(self:Audio Text Box Object Reference=@n7) -> (ValueTextBox:Editable Text Object Reference)
    Delegate:"Delegate (by ref)"=@n3
   ) -> (then:Exec) {is=title K2Node_AssignDelegate}
   Event PreConstruct()
  ]}
 GetDbValue {nodes [
   n1 = GetDbValue!() -> (then, LinValue:double=~) {is=title K2Node_FunctionEntry}
   n2 = Get LinToDbCurve() -> (LinToDbCurve:Curve Float Object Reference)
   > Is Valid(InputObject=@n2)
   > Return Node!(execute, DbValue:double=GetFloatValue!(self:Curve Float Object Reference=@n2, InTime:float=@n1) -> (ReturnValue:float))
  ]}
 GetLinValue {nodes [
   n1 = GetLinValue!() -> (then, DbValue:double=~) {is=title K2Node_FunctionEntry}
   n2 = Get DbToLinCurve() -> (DbToLinCurve:Curve Float Object Reference)
   > Is Valid(InputObject=@n2)
   > Return Node!(execute, LinValue:double=GetFloatValue(self=@n2, InTime=@n1))
  ]}
 OnValueChanged {nodes [OnValueChanged!() -> (then, NewValue:double=~) {is=title K2Node_FunctionEntry}]}
 SetTextValue {nodes [
   n1 = SetTextValue!() -> (then, NewValue:double=~) {is=title K2Node_FunctionEntry}
   > "SetText (Editable Text)"!(
    execute
    self:Editable Text Object Reference=Get ValueTextBox(self=Get AudioTextBox())
    InText:Text="To Text (Float)"(
     Value=@n1
     MinimumIntegralDigits=Get Minimum Integral Digits() -> (Minimum Integral Digits:int=0)
     MaximumIntegralDigits=Get Maximum Integral Digits() -> (Maximum Integral Digits:int=0)
     MinimumFractionalDigits=Get Minimum Fractional Digits() -> (Minimum Fractional Digits:int=0)
     MaximumFractionalDigits=Get Maximum Fractional Digits() -> (Maximum Fractional Digits:int=0)
    )
   ) -> (then)
  ]}
]
variables [
 DbToLinCurve {type=CurveFloat, default=/AudioWidgets/AudioControlCurves/AudioControl_DbLinCurveDefault}
 LinToDbCurve {type=CurveFloat, default=/AudioWidgets/AudioControlCurves/AudioControl_LinDbCurveDefault}
 Maximum Fractional Digits {type=integer, default=2}
 Maximum Integral Digits {type=integer, default=3}
 Minimum Fractional Digits {type=integer}
 Minimum Integral Digits {type=integer, default=1}
 OnValueChanged {}
]
widgets {
 info {inheritedWidgetCount=0, parentClass=&UserWidget, rootWidgetClass=&VerticalBox, widgetCount=5}
 widgets [
  VerticalBox_0 {widgetClassPath=&VerticalBox, widget=VerticalBox_0}
  AudioTextBox {
   widgetClassPath=/AudioWidgets/AudioTextBox/AudioTextBox
   bIsVariable=true
   parent=VerticalBox_0
   widget=AudioTextBox
   slot {class=VerticalBoxSlot, properties {horizontalAlignment=HAlign_Center, verticalAlignment=VAlign_Center, padding=<ltrb 0 2 0 2>}}
  }
  CanvasPanel_3 {widgetClassPath=&CanvasPanel, parent=VerticalBox_0, widget=CanvasPanel_3, slot {class=VerticalBoxSlot}}
  SliderShadow {
   widgetClassPath=&Image
   bIsVariable=true
   parent=CanvasPanel_3
   widget=SliderShadow
   properties {brush {
     bIsDynamicallyLoaded=false
     drawAs=Image
     imageType=NoImage
     mirroring=NoMirror
     resourceName=None
     resourceObject=&/AudioFader_Outer
     tiling=NoTile
     imageSize=<28 450>
     margin=<ltrb 0 0 0 0>
     outlineSettings {
      bUseBrushTransparency=false
      roundingType=HalfHeightRadius
      width=0
      color {colorUseRule=UseColor_Specified, specifiedColor=<rgba 0 0 0 0>}
      cornerRadii=<0 0 0 0>
     }
     tintColor {colorUseRule=UseColor_Specified, specifiedColor=<rgba 1 1 1 1>}
     uVRegion {bIsValid=false, max=<0 0>, min=<0 0>}
    }}
   slot {class=CanvasPanelSlot, properties {layoutData {alignment=<0.5 0>, anchors {maximum=<0.5 0>, minimum=<0.5 0>}, offsets=<ltrb 0 0 28 450>}}}
  }
  Slider {
   widgetClassPath=/Script/UMG.Slider
   bIsVariable=true
   parent=CanvasPanel_3
   widget=Slider
   properties {
    orientation=Orient_Vertical
    value=1
    renderTransform {angle=0, scale=<2 1>, shear=<0 0>, translation=<0 0>}
    widgetStyle {
     barThickness=2
     disabledBarImage {
      bIsDynamicallyLoaded=false
      drawAs=Image
      imageType=NoImage
      mirroring=NoMirror
      resourceName=None
      resourceObject=&/AudioFader_Inner
      tiling=NoTile
      imageSize=<10 432>
      margin=<ltrb 0 0 0 0>
      outlineSettings {
       bUseBrushTransparency=false
       roundingType=HalfHeightRadius
       width=0
       color {colorUseRule=UseColor_Specified, specifiedColor=<rgba 0 0 0 0>}
       cornerRadii=<0 0 0 0>
      }
      tintColor {colorUseRule=UseColor_Specified, specifiedColor=<rgba 0.5 0.5 0.5 1>}
      uVRegion {bIsValid=false, max=<0 0>, min=<0 0>}
     }
     disabledThumbImage {
      bIsDynamicallyLoaded=false
      drawAs=NoDrawType
      imageType=FullColor
      mirroring=NoMirror
      resourceName=../../../Engine/Content/Slate/Common/Button_Disabled.png
      resourceObject=null
      tiling=NoTile
      imageSize=<8 14>
      margin=<ltrb 0 0 0 0>
      outlineSettings {
       bUseBrushTransparency=false
       roundingType=HalfHeightRadius
       width=0
       color {colorUseRule=UseColor_Specified, specifiedColor=<rgba 0 0 0 0>}
       cornerRadii=<0 0 0 0>
      }
      tintColor {colorUseRule=UseColor_Specified, specifiedColor=<rgba 1 1 1 1>}
      uVRegion {bIsValid=false, max=<0 0>, min=<0 0>}
     }
     hoveredBarImage {
      bIsDynamicallyLoaded=false
      drawAs=Image
      imageType=NoImage
      mirroring=NoMirror
      resourceName=None
      resourceObject=&AudioFader_Inner
      tiling=NoTile
      imageSize=<10 432>
      margin=<ltrb 0 0 0 0>
      outlineSettings {
       bUseBrushTransparency=false
       roundingType=HalfHeightRadius
       width=0
       color {colorUseRule=UseColor_Specified, specifiedColor=<rgba 0 0 0 0>}
       cornerRadii=<0 0 0 0>
      }
      tintColor {colorUseRule=UseColor_Specified, specifiedColor=<rgba 1 1 1 1>}
      uVRegion {bIsValid=false, max=<0 0>, min=<0 0>}
     }
     hoveredThumbImage {
      bIsDynamicallyLoaded=false
      drawAs=Image
      imageType=FullColor
      mirroring=NoMirror
      resourceName=../../../Engine/Content/Slate/Common/Button.png
      resourceObject=&/AudioFader_Selected
      tiling=NoTile
      imageSize=<24 12>
      margin=<ltrb 0 0 0 0>
      outlineSettings {
       bUseBrushTransparency=false
       roundingType=HalfHeightRadius
       width=0
       color {colorUseRule=UseColor_Specified, specifiedColor=<rgba 0 0 0 0>}
       cornerRadii=<0 0 0 0>
      }
      tintColor {colorUseRule=UseColor_Specified, specifiedColor=<rgba 1 1 1 1>}
      uVRegion {bIsValid=false, max=<0 0>, min=<0 0>}
     }
     normalBarImage {
      bIsDynamicallyLoaded=false
      drawAs=Image
      imageType=NoImage
      mirroring=NoMirror
      resourceName=None
      resourceObject=&AudioFader_Inner
      tiling=NoTile
      imageSize=<28 450>
      margin=<ltrb 0 0 0 0>
      outlineSettings {
       bUseBrushTransparency=false
       roundingType=HalfHeightRadius
       width=0
       color {colorUseRule=UseColor_Specified, specifiedColor=<rgba 0 0 0 0>}
       cornerRadii=<0 0 0 0>
      }
      tintColor {colorUseRule=UseColor_Specified, specifiedColor=<rgba 0 0 0 1>}
      uVRegion {bIsValid=false, max=<0 0>, min=<0 0>}
     }
     normalThumbImage {
      bIsDynamicallyLoaded=false
      drawAs=Image
      imageType=FullColor
      mirroring=NoMirror
      resourceName=../../../Engine/Content/Slate/Common/Button.png
      resourceObject=&/AudioFader_Normal
      tiling=NoTile
      imageSize=<24 12>
      margin=<ltrb 0 0 0 0>
      outlineSettings {
       bUseBrushTransparency=false
       roundingType=HalfHeightRadius
       width=0
       color {colorUseRule=UseColor_Specified, specifiedColor=<rgba 0 0 0 0>}
       cornerRadii=<0 0 0 0>
      }
      tintColor {colorUseRule=UseColor_Specified, specifiedColor=<rgba 1 1 1 1>}
      uVRegion {bIsValid=false, max=<0 0>, min=<0 0>}
     }
    }
   }
   slot {class=CanvasPanelSlot, properties {layoutData {alignment=<0.5 0>, anchors {maximum=<0.5 0>, minimum=<0.5 0>}, offsets=<ltrb 0 3 10 444>}}}
  }
 ]
}
